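# Sound cue for DadBro's DBNO (down but not out) grunt vocals, reusing Storm King pain grunt recordings.
# A quality-level switch sends the first two levels to an equal-weight random pick of thirteen grunts and the last two to a pick of three.
# A Modulator at the root varies the result with default settings.
# It plays in the boss-fight VOX sound class using the Storm King vocal attenuation asset.
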SoundCue /Game/Sounds/Apollo/Events/Dadbro/Efforts/DadBro_DBNO_Grunts_Vox_Cue
root=Modulator_0
nodes [
 {id=Modulator_0, children [QualityLevel_0]}
 {id=QualityLevel_0, children [Random_1,Random_1,Random_0,Random_0]}
 {id=Random_0, children [WavePlayer_0, WavePlayer_1, WavePlayer_15], properties {weights [1,1,1]}}
 {
  id=Random_1
  children [
   WavePlayer_2
   WavePlayer_3
   WavePlayer_4
   WavePlayer_5
   WavePlayer_6
   WavePlayer_7
   WavePlayer_8
   WavePlayer_9
   WavePlayer_10
   WavePlayer_11
   WavePlayer_12
   WavePlayer_13
   WavePlayer_14
  ]
  properties {weights [1,1,1,1,1,1,1,1,1,1,1,1,1]}
 }
 {id=WavePlayer_0, properties {soundWaveAssetPtr=/Game/Sounds/Fort_Monster_Storm_King/Efforts/StormKing_DBNO_Pain_Grunt_10}}
 {id=WavePlayer_1, properties {soundWaveAssetPtr=/Game/Sounds/Fort_Monster_Storm_King/Efforts/StormKing_DBNO_Pain_Grunt_11}}
 {id=WavePlayer_10, properties {soundWaveAssetPtr=/Game/Sounds/Fort_Monster_Storm_King/Efforts/StormKing_DBNO_Pain_Grunt_05}}
 {id=WavePlayer_11, properties {soundWaveAssetPtr=/Game/Sounds/Fort_Monster_Storm_King/Efforts/StormKing_DBNO_Pain_Grunt_06}}
 {id=WavePlayer_12, properties {soundWaveAssetPtr=/Game/Sounds/Fort_Monster_Storm_King/Efforts/StormKing_DBNO_Pain_Grunt_07}}
 {id=WavePlayer_13, properties {soundWaveAssetPtr=/Game/Sounds/Fort_Monster_Storm_King/Efforts/StormKing_DBNO_Pain_Grunt_08}}
 {id=WavePlayer_14, properties {soundWaveAssetPtr=/Game/Sounds/Fort_Monster_Storm_King/Efforts/StormKing_DBNO_Pain_Grunt_09}}
 {id=WavePlayer_15, properties {soundWaveAssetPtr=/Game/Sounds/Fort_Monster_Storm_King/Efforts/StormKing_DBNO_Pain_Grunt_12}}
 {id=WavePlayer_2, properties {soundWaveAssetPtr=&StormKing_DBNO_Pain_Grunt_10}}
 {id=WavePlayer_3, properties {soundWaveAssetPtr=&StormKing_DBNO_Pain_Grunt_11}}
 {id=WavePlayer_4, properties {soundWaveAssetPtr=&StormKing_DBNO_Pain_Grunt_12}}
 {id=WavePlayer_5, properties {soundWaveAssetPtr=/Game/Sounds/Fort_Monster_Storm_King/Efforts/StormKing_DBNO_Pain_Grunt_13}}
 {id=WavePlayer_6, properties {soundWaveAssetPtr=/Game/Sounds/Fort_Monster_Storm_King/Efforts/StormKing_DBNO_Pain_Grunt_01}}
 {id=WavePlayer_7, properties {soundWaveAssetPtr=/Game/Sounds/Fort_Monster_Storm_King/Efforts/StormKing_DBNO_Pain_Grunt_02}}
 {id=WavePlayer_8, properties {soundWaveAssetPtr=/Game/Sounds/Fort_Monster_Storm_King/Efforts/StormKing_DBNO_Pain_Grunt_03}}
 {id=WavePlayer_9, properties {soundWaveAssetPtr=/Game/Sounds/Fort_Monster_Storm_King/Efforts/StormKing_DBNO_Pain_Grunt_04}}
]
properties {
 attenuationSettings=/Game/Sounds/Attenuation/Fort_StormKingVox_Att
 maxDistance=20000
 soundClassObject=/Game/Sounds/FortSoundClassesAndModes/Fort_BossFight_SoundClass/Fort_BossFight_VOX
 attenuationOverrides {bAttenuate=false, lPFRadiusMin=17, spatializationAlgorithm=SPATIALIZATION_HRTF, stereoSpread=1000}
}
