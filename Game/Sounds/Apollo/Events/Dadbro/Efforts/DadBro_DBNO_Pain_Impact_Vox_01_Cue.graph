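# Sound Cue for DadBro's down-but-not-out pain impact vocal effort, built from Storm King voice recordings.
# A Random node picks between two equally weighted waves, is excluded from branch culling, and feeds a Modulator at default settings.
# The cue plays in the boss fight VOX sound class and references the Storm King vox attenuation asset.
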SoundCue /Game/Sounds/Apollo/Events/Dadbro/Efforts/DadBro_DBNO_Pain_Impact_Vox_01_Cue
root=Modulator_0
nodes [
 {id=Modulator_0, children [Random_0]}
 {id=Random_0, children [WavePlayer_0,WavePlayer_1], properties {bShouldExcludeFromBranchCulling=true, weights [1,1]}}
 {id=WavePlayer_0, properties {soundWaveAssetPtr=/Game/Sounds/Fort_Monster_Storm_King/Efforts/StormKing_DBNO_Pain_Impact_Vox_01}}
 {id=WavePlayer_1, properties {soundWaveAssetPtr=/Game/Sounds/Fort_Monster_Storm_King/Efforts/StormKing_DBNO_Pain_Impact_Vox_02}}
]
properties {
 attenuationSettings=/Game/Sounds/Attenuation/Fort_StormKingVox_Att
 maxDistance=20000
 soundClassObject=/Game/Sounds/FortSoundClassesAndModes/Fort_BossFight_SoundClass/Fort_BossFight_VOX
 attenuationOverrides {bAttenuate=false, lPFRadiusMin=17, spatializationAlgorithm=SPATIALIZATION_HRTF, stereoSpread=1000}
}
