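# Sound cue for the DadBro death vocal effort, which reuses Storm King death vox recordings.
# A Random node picks one of two equally weighted wave players and feeds it through a Modulator at the root.
# It plays under the boss-fight VOX sound class and references the shared Storm King vox attenuation asset.
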SoundCue /Game/Sounds/Apollo/Events/Dadbro/Efforts/DadBro_Death_Vox_01_Cue
root=Modulator_0
nodes [
 {id=Modulator_0, children [Random_1]}
 {id=Random_1, children [WavePlayer_2,WavePlayer_3], properties {weights [1,1]}}
 {id=WavePlayer_2, properties {soundWaveAssetPtr=/Game/Sounds/Fort_Monster_Storm_King/Efforts/StormKing_Death_Vox_v2_01}}
 {id=WavePlayer_3, properties {soundWaveAssetPtr=/Game/Sounds/Fort_Monster_Storm_King/Efforts/StormKing_Death_Vox_v2_02}}
]
properties {
 attenuationSettings=/Game/Sounds/Attenuation/Fort_StormKingVox_Att
 maxDistance=20000
 soundClassObject=/Game/Sounds/FortSoundClassesAndModes/Fort_BossFight_SoundClass/Fort_BossFight_VOX
 attenuationOverrides {bAttenuate=false, lPFRadiusMin=17, spatializationAlgorithm=SPATIALIZATION_HRTF, stereoSpread=1000}
}
