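# Sound cue for the DadBro (Storm King) boss's vocal reaction to being hit, using back hit-react clips.
# A Random node picks one of three equally weighted voice waves.
# Playback uses the shared Storm King vox attenuation and the boss-fight VOX sound class, with a slight volume boost.
# A concurrency override prevents new instances once the limit is reached, and the debug flag is left on.
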SoundCue /Game/Sounds/Apollo/Events/Dadbro/Efforts/DadBro_Hit_Back_Vox_Cue
root=Random_0
nodes [
 {id=Random_0, children [WavePlayer_0,WavePlayer_1,WavePlayer_3], properties {weights [1,1,1]}}
 {id=WavePlayer_0, properties {soundWaveAssetPtr=&/StormKing_HitReact_Back_Vox_01}}
 {id=WavePlayer_1, properties {soundWaveAssetPtr=&/StormKing_HitReact_Back_Vox_02}}
 {id=WavePlayer_3, properties {soundWaveAssetPtr=&/StormKing_HitReact_Back_Vox_03}}
]
properties {
 attenuationSettings=/Game/Sounds/Attenuation/Fort_StormKingVox_Att
 bDebug=true
 bOverrideConcurrency=true
 maxDistance=20000
 soundClassObject=/Game/Sounds/FortSoundClassesAndModes/Fort_BossFight_SoundClass/Fort_BossFight_VOX
 volumeMultiplier=1.2
 attenuationOverrides {
  bAttenuate=false
  distanceAlgorithm=NaturalSound
  falloffDistance=5000
  lPFRadiusMin=17
  spatializationAlgorithm=SPATIALIZATION_HRTF
  stereoSpread=1000
  attenuationShapeExtents {x=1000}
 }
 concurrencyOverrides {resolutionRule=PreventNew}
}
