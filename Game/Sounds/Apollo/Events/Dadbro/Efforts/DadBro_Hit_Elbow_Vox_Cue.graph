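# Sound cue for the DadBro (Storm King) boss's vocal effort when reacting to a hit on the elbow.
# A Random node picks one of three equally weighted elbow hit-react voice waves.
# Playback uses the shared Storm King vox attenuation and the boss fight VOX sound class, with a slight volume boost.
# Its own concurrency override blocks new instances once the limit is reached, and the cue's debug flag is left on.
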SoundCue /Game/Sounds/Apollo/Events/Dadbro/Efforts/DadBro_Hit_Elbow_Vox_Cue
root=Random_1
nodes [
 {id=Random_1, children [WavePlayer_2,WavePlayer_4,WavePlayer_5], properties {weights [1,1,1]}}
 {id=WavePlayer_2, properties {soundWaveAssetPtr=&/StormKing_HitReact_Elbow_Vox_01}}
 {id=WavePlayer_4, properties {soundWaveAssetPtr=&/StormKing_HitReact_Elbow_Vox_02}}
 {id=WavePlayer_5, properties {soundWaveAssetPtr=&/StormKing_HitReact_Elbow_Vox_03}}
]
properties {
 attenuationSettings=/Game/Sounds/Attenuation/Fort_StormKingVox_Att
 bDebug=true
 bOverrideConcurrency=true
 maxDistance=20000
 soundClassObject=/Game/Sounds/FortSoundClassesAndModes/Fort_BossFight_SoundClass/Fort_BossFight_VOX
 volumeMultiplier=1.2
 attenuationOverrides {
  bAttenuate=false
  distanceAlgorithm=NaturalSound
  falloffDistance=5000
  lPFRadiusMin=17
  spatializationAlgorithm=SPATIALIZATION_HRTF
  stereoSpread=1000
  attenuationShapeExtents {x=1000}
 }
 concurrencyOverrides {resolutionRule=PreventNew}
}
