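# Sound cue for the DadBro spawn-in vocal in the Apollo event audio, routed to the boss fight VOX sound class.
# A single wave player plays a Storm King spawn-in vox wave, with no randomization or mixing.
# Overridden attenuation turns off volume falloff with distance and selects HRTF spatialization.
# Overridden concurrency uses the PreventNew rule, and the cue's debug flag is on.
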SoundCue /Game/Sounds/Apollo/Events/Dadbro/Efforts/DadBro_Spawn_In_Vox_Cue
root=WavePlayer_4
nodes [{id=WavePlayer_4, properties {soundWaveAssetPtr=&/StormKing_SpawnIn_Vox_02}}]
properties {
 bDebug=true
 bOverrideAttenuation=true
 bOverrideConcurrency=true
 maxDistance=2097152
 soundClassObject=/Game/Sounds/FortSoundClassesAndModes/Fort_BossFight_SoundClass/Fort_BossFight_VOX
 volumeMultiplier=1
 attenuationOverrides {
  bAttenuate=false
  distanceAlgorithm=NaturalSound
  falloffDistance=5000
  lPFRadiusMin=17
  spatializationAlgorithm=SPATIALIZATION_HRTF
  stereoSpread=1000
  attenuationShapeExtents {x=1000}
 }
 concurrencyOverrides {resolutionRule=PreventNew}
}
